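# Physics asset for the M_MED_LatteStir male cosmetic body, providing secondary motion for the sleeves, a bandolier strap and a purse.
# Capsules on the spine and arms form a locked body frame, with four swing-limited sleeve flaps per forearm.
# Front and back bandolier chains hang from a root on spine_05 and close into a loop at their ends.
# Both ends also link to the last segment of a swing-only purse chain anchored to spine_04.
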PhysicsAsset /BRCosmetics/Characters/Player/Male/Medium/Bodies/M_MED_LatteStir/Meshes/M_MED_LatteStir_FE_Physics
bodies {
 base_bandolier_bk_01_l [{shapeType=Capsule, length=0.505, radius=0.505, shapeName=base_bandolier_bk_01_l_capsule, rotation=<rot 0 -90 0>}]
 base_bandolier_root [{shapeType=Capsule, length=0.505, radius=0.505, shapeName=base_bandolier_root_capsule, rotation=<rot 0 -90 0>}]
 base_purse_l [{shapeType=Capsule, length=0.505, radius=0.505, shapeName=base_purse_l_capsule, rotation=<rot 0 -90 0>}]
 base_sleeve_l_A [{shapeType=Capsule, length=2, radius=1, shapeName=base_sleeve_l_A_capsule, rotation=<rot 90 0 0>}]
 base_sleeve_l_B [{shapeType=Capsule, length=2, radius=1, shapeName=base_sleeve_l_B_capsule, rotation=<rot 90 0 0>}]
 base_sleeve_l_C [{shapeType=Capsule, length=2, radius=1, shapeName=base_sleeve_l_C_capsule, rotation=<rot 90 0 0>}]
 base_sleeve_l_D [{shapeType=Capsule, length=2, radius=1, shapeName=base_sleeve_l_D_capsule, rotation=<rot 90 0 0>}]
 base_sleeve_r_A [{shapeType=Capsule, length=2, radius=1, shapeName=base_sleeve_r_A_capsule, rotation=<rot 90 0 0>}]
 base_sleeve_r_B [{shapeType=Capsule, length=2, radius=1, shapeName=base_sleeve_r_B_capsule, rotation=<rot 90 0 0>}]
 base_sleeve_r_C [{shapeType=Capsule, length=2, radius=1, shapeName=base_sleeve_r_C_capsule, rotation=<rot 90 0 0>}]
 base_sleeve_r_D [{shapeType=Capsule, length=2, radius=1, shapeName=base_sleeve_r_D_capsule, rotation=<rot 90 0 0>}]
 dyn_bandolier_bk_01_l [{shapeType=Capsule, length=0.505, radius=0.505, shapeName=base_bandolier_bk_01_l_capsule, rotation=<rot 0 -90 0>}]
 dyn_bandolier_bk_02_l [{
   shapeType=Capsule
   length=8.391
   radius=1
   shapeName=dyn_bandolier_bk_02_l_capsule
   center=<4.1929 -0.3466 -0.8395>
   rotation=<rot 80.2838 -0.0012 -1.99>
  }]
 dyn_bandolier_bk_03_l [{
   shapeType=Capsule
   length=8
   radius=1
   shapeName=dyn_bandolier_bk_02_l_capsule
   center=<4.6598 -0.5344 -0.2655>
   rotation=<rot 89.0596 -0.7919 -0.0035>
  }]
 dyn_bandolier_bk_04_l [{shapeType=Capsule, length=8, radius=1, shapeName=dyn_bandolier_bk_02_l_capsule, center=<4.1207 -0.0135 0>, rotation=<rot 90 0 0>}]
 dyn_bandolier_bk_05_l [{
   shapeType=Capsule
   length=5.069
   radius=1.3394
   shapeName=dyn_bandolier_bk_02_l_capsule
   center=<3.5897 3.0174 0.4216>
   rotation=<rot 83.9915 -135.5858 166.6098>
  }]
 dyn_bandolier_fr_01_l [{
   shapeType=Capsule
   length=6.36
   radius=1.0741
   shapeName=dyn_bandolier_fr_01_l_capsule
   center=<4.2678 0.1627 -0.274>
   rotation=<rot 0 89.6864 -91.0589>
  }]
 dyn_bandolier_fr_02_l [{
   shapeType=Capsule
   length=6.9726
   radius=1.0787
   shapeName=dyn_bandolier_fr_02_l_capsule
   center=<4.8292 0.5941 -0.2189>
   rotation=<rot 88.7332 -13.3512 -12.9631>
  }]
 dyn_bandolier_fr_03_l [{
   shapeType=Capsule
   length=8.2239
   radius=1.03
   shapeName=dyn_bandolier_fr_03_l_capsule
   center=<4.4004 0.7248 -0.2487>
   rotation=<rot 86.87 -75.6641 -73.8903>
  }]
 dyn_bandolier_fr_04_l [{
   shapeType=Capsule
   length=8.1468
   radius=1.1011
   shapeName=dyn_bandolier_fr_04_l_capsule
   center=<4.3911 0.722 -0.2292>
   rotation=<rot 89.429 -61.3753 -65.0667>
  }]
 dyn_bandolier_fr_05_l [{
   shapeType=Capsule
   length=5.5107
   radius=1.0981
   shapeName=dyn_bandolier_fr_05_l_capsule
   center=<3.3502 -2.2284 0.8037>
   rotation=<rot 76.6733 132.1934 177.3815>
  }]
 dyn_purse_01_l [{shapeType=Capsule, length=6.8395, radius=2.119, shapeName=dyn_purse_01_l_capsule, center=<3 0 0>, rotation=<rot 90 0 0>}]
 dyn_purse_02_l [{shapeType=Capsule, length=6.8395, radius=2.119, shapeName=dyn_purse_02_l_capsule, center=<3 0 0>, rotation=<rot 90 0 0>}]
 dyn_purse_03_l [{
   shapeType=Capsule
   length=7.056
   radius=1.944
   shapeName=dyn_purse_03_l_capsule
   center=<1.4481 0 0.1404>
   rotation=<rot 75.322 -179.9998 90.0002>
  }]
 dyn_sleeve_l_A [{shapeType=Capsule, length=9, radius=3, shapeName=dyn_sleeve_l_A_capsule, center=<5 0 3>, rotation=<rot 85.9815 0.007 0.0078>}]
 dyn_sleeve_l_B [{shapeType=Capsule, length=9, radius=3, shapeName=dyn_sleeve_l_B_capsule, center=<5 0 3>, rotation=<rot 85.9815 0.007 0.0078>}]
 dyn_sleeve_l_C [{shapeType=Capsule, length=9, radius=3, shapeName=dyn_sleeve_l_C_capsule, center=<5 0 3>, rotation=<rot 85.9815 0.007 0.0078>}]
 dyn_sleeve_l_D [{shapeType=Capsule, length=9, radius=3, shapeName=dyn_sleeve_l_D_capsule, center=<5 0 3>, rotation=<rot 85.9815 0.007 0.0078>}]
 dyn_sleeve_r_A [{shapeType=Capsule, length=9, radius=3, shapeName=dyn_sleeve_r_A_capsule, center=<5 0 3>, rotation=<rot 85.9815 0.007 0.0078>}]
 dyn_sleeve_r_B [{shapeType=Capsule, length=9, radius=3, shapeName=dyn_sleeve_r_B_capsule, center=<5 0 3>, rotation=<rot 85.9815 0.007 0.0078>}]
 dyn_sleeve_r_C [{shapeType=Capsule, length=9, radius=3, shapeName=dyn_sleeve_r_C_capsule, center=<5 0 3>, rotation=<rot 85.9815 0.007 0.0078>}]
 dyn_sleeve_r_D [{shapeType=Capsule, length=9, radius=3, shapeName=dyn_sleeve_r_D_capsule, center=<5 0 3>, rotation=<rot 85.9815 0.007 0.0078>}]
 lowerarm_l [{shapeType=Capsule, length=19.8537, radius=4, shapeName=lowerarm_l_capsule, center=<11.5381 -0.1214 0>, rotation=<rot 90 0.0473 0>}]
 lowerarm_r [{shapeType=Capsule, length=19.8537, radius=4, shapeName=lowerarm_l_capsule, center=<-11.5381 -0.1214 0>, rotation=<rot 90 -0.0473 0>}]
 spine_01 [{shapeType=Capsule, length=6.0919, radius=10.5683, shapeName=spine_05_capsule}]
 spine_02 [{shapeType=Capsule, length=1.8988, radius=11.5512, shapeName=spine_05_capsule}]
 spine_03 [
  {shapeType=Capsule, length=8.4197, radius=10.5666, shapeName=spine_05_capsule, center=<-1.1846 0 0>}
  {shapeType=Capsule, length=4.3942, radius=7.3552, shapeName=spine_03_capsule, center=<4.2569 3.4879 0>}
  {shapeType=Capsule, length=2.0802, radius=5.7765, shapeName=spine_03_capsule, center=<-2.3233 0 -8.0411>}
 ]
 spine_04 [{shapeType=Capsule, length=5.3616, radius=10.5585, shapeName=spine_05_capsule, center=<0 0.7023 0>}]
 spine_05 [{shapeType=Capsule, length=11.5769, radius=8.8031, shapeName=spine_05_capsule, center=<0 0.345 0>}]
 upperarm_l [{shapeType=Capsule, length=24, radius=5, shapeName=upperarm_l_capsule, center=<9.3223 0 -2.5>, rotation=<rot 90 0 0>}]
 upperarm_r [{shapeType=Capsule, length=24, radius=5, shapeName=upperarm_l_capsule, center=<-9.3223 0 2.5>, rotation=<rot 90 0 0>}]
}
constraints [
 {bone1Name=dyn_sleeve_l_A, bone2Name=base_sleeve_l_A, swing1LimitDegrees=5, swing2LimitDegrees=15, twistLimitDegrees=45, twistMotion=Locked}
 {bone1Name=dyn_sleeve_l_B, bone2Name=base_sleeve_l_B, swing1LimitDegrees=5, swing2LimitDegrees=15, twistLimitDegrees=45, twistMotion=Locked}
 {bone1Name=dyn_sleeve_l_C, bone2Name=base_sleeve_l_C, swing1LimitDegrees=5, swing2LimitDegrees=5, twistLimitDegrees=45, twistMotion=Locked}
 {bone1Name=dyn_sleeve_l_D, bone2Name=base_sleeve_l_D, swing1LimitDegrees=5, swing2LimitDegrees=15, twistLimitDegrees=45, twistMotion=Locked}
 {bone1Name=dyn_sleeve_r_A, bone2Name=base_sleeve_r_A, swing1LimitDegrees=5, swing2LimitDegrees=15, twistLimitDegrees=45, twistMotion=Locked}
 {bone1Name=dyn_sleeve_r_B, bone2Name=base_sleeve_r_B, swing1LimitDegrees=5, swing2LimitDegrees=15, twistLimitDegrees=45, twistMotion=Locked}
 {bone1Name=dyn_sleeve_r_C, bone2Name=base_sleeve_r_C, swing1LimitDegrees=5, swing2LimitDegrees=5, twistLimitDegrees=45, twistMotion=Locked}
 {bone1Name=dyn_sleeve_r_D, bone2Name=base_sleeve_r_D, swing1LimitDegrees=5, swing2LimitDegrees=15, twistLimitDegrees=45, twistMotion=Locked}
 {
  bone1Name=base_sleeve_l_A
  bone2Name=lowerarm_l
  swing1LimitDegrees=45
  swing1Motion=Locked
  swing2LimitDegrees=45
  swing2Motion=Locked
  twistLimitDegrees=45
  twistMotion=Locked
 }
 {
  bone1Name=base_sleeve_l_B
  bone2Name=lowerarm_l
  swing1LimitDegrees=45
  swing1Motion=Locked
  swing2LimitDegrees=45
  swing2Motion=Locked
  twistLimitDegrees=45
  twistMotion=Locked
 }
 {
  bone1Name=base_sleeve_l_C
  bone2Name=lowerarm_l
  swing1LimitDegrees=45
  swing1Motion=Locked
  swing2LimitDegrees=45
  swing2Motion=Locked
  twistLimitDegrees=45
  twistMotion=Locked
 }
 {
  bone1Name=base_sleeve_l_D
  bone2Name=lowerarm_l
  swing1LimitDegrees=45
  swing1Motion=Locked
  swing2LimitDegrees=45
  swing2Motion=Locked
  twistLimitDegrees=45
  twistMotion=Locked
 }
 {
  bone1Name=base_sleeve_r_A
  bone2Name=lowerarm_r
  swing1LimitDegrees=45
  swing1Motion=Locked
  swing2LimitDegrees=45
  swing2Motion=Locked
  twistLimitDegrees=45
  twistMotion=Locked
 }
 {
  bone1Name=base_sleeve_r_B
  bone2Name=lowerarm_r
  swing1LimitDegrees=45
  swing1Motion=Locked
  swing2LimitDegrees=45
  swing2Motion=Locked
  twistLimitDegrees=45
  twistMotion=Locked
 }
 {
  bone1Name=base_sleeve_r_C
  bone2Name=lowerarm_r
  swing1LimitDegrees=45
  swing1Motion=Locked
  swing2LimitDegrees=45
  swing2Motion=Locked
  twistLimitDegrees=45
  twistMotion=Locked
 }
 {
  bone1Name=base_sleeve_r_D
  bone2Name=lowerarm_r
  swing1LimitDegrees=45
  swing1Motion=Locked
  swing2LimitDegrees=45
  swing2Motion=Locked
  twistLimitDegrees=45
  twistMotion=Locked
 }
 {
  bone1Name=spine_02
  bone2Name=spine_01
  swing1LimitDegrees=45
  swing1Motion=Locked
  swing2LimitDegrees=45
  swing2Motion=Locked
  twistLimitDegrees=45
  twistMotion=Locked
 }
 {
  bone1Name=spine_03
  bone2Name=spine_02
  swing1LimitDegrees=45
  swing1Motion=Locked
  swing2LimitDegrees=45
  swing2Motion=Locked
  twistLimitDegrees=45
  twistMotion=Locked
 }
 {
  bone1Name=spine_04
  bone2Name=spine_03
  swing1LimitDegrees=45
  swing1Motion=Locked
  swing2LimitDegrees=45
  swing2Motion=Locked
  twistLimitDegrees=45
  twistMotion=Locked
 }
 {
  bone1Name=spine_05
  bone2Name=spine_04
  swing1LimitDegrees=45
  swing1Motion=Locked
  swing2LimitDegrees=45
  swing2Motion=Locked
  twistLimitDegrees=45
  twistMotion=Locked
 }
 {
  bone1Name=upperarm_l
  bone2Name=spine_05
  swing1LimitDegrees=45
  swing1Motion=Locked
  swing2LimitDegrees=45
  swing2Motion=Locked
  twistLimitDegrees=45
  twistMotion=Locked
 }
 {
  bone1Name=upperarm_r
  bone2Name=spine_05
  swing1LimitDegrees=45
  swing1Motion=Locked
  swing2LimitDegrees=45
  swing2Motion=Locked
  twistLimitDegrees=45
  twistMotion=Locked
 }
 {
  bone1Name=lowerarm_l
  bone2Name=upperarm_l
  swing1LimitDegrees=45
  swing1Motion=Locked
  swing2LimitDegrees=45
  swing2Motion=Locked
  twistLimitDegrees=45
  twistMotion=Locked
 }
 {
  bone1Name=lowerarm_r
  bone2Name=upperarm_r
  swing1LimitDegrees=45
  swing1Motion=Locked
  swing2LimitDegrees=45
  swing2Motion=Locked
  twistLimitDegrees=45
  twistMotion=Locked
 }
 {
  bone1Name=dyn_bandolier_bk_01_l
  bone2Name=base_bandolier_bk_01_l
  swing1LimitDegrees=45
  swing1Motion=Free
  swing2LimitDegrees=45
  swing2Motion=Free
  twistLimitDegrees=45
  twistMotion=Free
 }
 {
  bone1Name=dyn_bandolier_bk_02_l
  bone2Name=dyn_bandolier_bk_01_l
  swing1LimitDegrees=8
  swing2LimitDegrees=10
  twistLimitDegrees=45
  twistMotion=Locked
 }
 {
  bone1Name=dyn_bandolier_bk_03_l
  bone2Name=dyn_bandolier_bk_02_l
  swing1LimitDegrees=45
  swing1Motion=Locked
  swing2LimitDegrees=10
  twistLimitDegrees=45
  twistMotion=Locked
 }
 {
  bone1Name=dyn_bandolier_bk_04_l
  bone2Name=dyn_bandolier_bk_03_l
  swing1LimitDegrees=45
  swing1Motion=Locked
  swing2LimitDegrees=20
  twistLimitDegrees=45
  twistMotion=Locked
 }
 {
  bone1Name=dyn_bandolier_bk_05_l
  bone2Name=dyn_bandolier_bk_04_l
  swing1LimitDegrees=45
  swing1Motion=Locked
  swing2LimitDegrees=20
  twistLimitDegrees=45
  twistMotion=Locked
 }
 {
  bone1Name=dyn_bandolier_fr_01_l
  bone2Name=base_bandolier_root
  swing1LimitDegrees=45
  swing1Motion=Free
  swing2LimitDegrees=45
  swing2Motion=Free
  twistLimitDegrees=45
  twistMotion=Free
 }
 {
  bone1Name=dyn_bandolier_fr_02_l
  bone2Name=dyn_bandolier_fr_01_l
  swing1LimitDegrees=45
  swing1Motion=Free
  swing2LimitDegrees=45
  swing2Motion=Free
  twistLimitDegrees=45
  twistMotion=Free
 }
 {
  bone1Name=dyn_bandolier_fr_03_l
  bone2Name=dyn_bandolier_fr_02_l
  swing1LimitDegrees=20
  swing2LimitDegrees=10
  twistLimitDegrees=45
  twistMotion=Locked
 }
 {
  bone1Name=dyn_bandolier_fr_04_l
  bone2Name=dyn_bandolier_fr_03_l
  swing1LimitDegrees=10
  swing2LimitDegrees=15
  twistLimitDegrees=45
  twistMotion=Locked
 }
 {
  bone1Name=dyn_bandolier_fr_05_l
  bone2Name=dyn_bandolier_fr_04_l
  swing1LimitDegrees=45
  swing1Motion=Locked
  swing2LimitDegrees=20
  twistLimitDegrees=45
  twistMotion=Locked
 }
 {
  bone1Name=base_bandolier_root
  bone2Name=spine_05
  swing1LimitDegrees=45
  swing1Motion=Free
  swing2LimitDegrees=45
  swing2Motion=Free
  twistLimitDegrees=45
  twistMotion=Free
 }
 {
  bone1Name=base_bandolier_bk_01_l
  bone2Name=base_bandolier_root
  swing1LimitDegrees=45
  swing1Motion=Free
  swing2LimitDegrees=45
  swing2Motion=Free
  twistLimitDegrees=45
  twistMotion=Free
 }
 {
  bone1Name=dyn_bandolier_bk_05_l
  bone2Name=dyn_bandolier_fr_05_l
  swing1LimitDegrees=45
  swing1Motion=Free
  swing2LimitDegrees=45
  swing2Motion=Free
  twistLimitDegrees=45
  twistMotion=Free
 }
 {
  bone1Name=dyn_purse_03_l
  bone2Name=dyn_purse_02_l
  swing1LimitDegrees=45
  swing1Motion=Locked
  swing2LimitDegrees=45
  twistLimitDegrees=45
  twistMotion=Locked
 }
 {
  bone1Name=dyn_purse_03_l
  bone2Name=dyn_bandolier_bk_05_l
  swing1LimitDegrees=45
  swing1Motion=Free
  swing2LimitDegrees=45
  swing2Motion=Free
  twistLimitDegrees=45
  twistMotion=Free
 }
 {
  bone1Name=dyn_purse_03_l
  bone2Name=dyn_bandolier_fr_05_l
  swing1LimitDegrees=45
  swing1Motion=Free
  swing2LimitDegrees=45
  swing2Motion=Free
  twistLimitDegrees=45
  twistMotion=Free
 }
 {
  bone1Name=dyn_purse_02_l
  bone2Name=dyn_purse_01_l
  swing1LimitDegrees=45
  swing1Motion=Locked
  swing2LimitDegrees=45
  twistLimitDegrees=45
  twistMotion=Locked
 }
 {
  bone1Name=dyn_purse_01_l
  bone2Name=base_purse_l
  swing1LimitDegrees=45
  swing1Motion=Locked
  swing2LimitDegrees=45
  twistLimitDegrees=45
  twistMotion=Locked
 }
 {
  bone1Name=base_purse_l
  bone2Name=spine_04
  swing1LimitDegrees=45
  swing1Motion=Free
  swing2LimitDegrees=45
  swing2Motion=Free
  twistLimitDegrees=45
  twistMotion=Free
 }
]
