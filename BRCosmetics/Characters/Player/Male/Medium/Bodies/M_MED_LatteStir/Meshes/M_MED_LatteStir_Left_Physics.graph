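# Physics asset for secondary motion of the left calf pants on the LatteStir medium male player body cosmetic.
# Capsules on the pelvis, thighs, calves and feet form a lower-body rig with all angular motion locked.
# Four pant chains (A–D) each hang from a small base capsule locked to calf_l, followed by two dynamic capsules.
# Dynamic joints lock twist but not swing, with small swing angles tuned per chain.
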PhysicsAsset /BRCosmetics/Characters/Player/Male/Medium/Bodies/M_MED_LatteStir/Meshes/M_MED_LatteStir_Left_Physics
bodies {
 base_calf_pants_l_A [{shapeType=Capsule, length=2, radius=1, shapeName=base_calf_pants_l_A_capsule}]
 base_calf_pants_l_B [{shapeType=Capsule, length=2, radius=1, shapeName=base_calf_pants_l_B_capsule}]
 base_calf_pants_l_C [{shapeType=Capsule, length=2, radius=1, shapeName=base_calf_pants_l_C_capsule}]
 base_calf_pants_l_D [{shapeType=Capsule, length=2, radius=1, shapeName=base_calf_pants_l_D_capsule}]
 calf_l [{shapeType=Capsule, length=33, radius=8, shapeName=thigh_l_capsule, center=<-22 0 0>, rotation=<rot 90 -1.4996 0>}]
 calf_r [{shapeType=Capsule, length=33, radius=8, shapeName=thigh_l_capsule, center=<22 0 0>, rotation=<rot 90 -1.4996 0>}]
 dyn_calf_pants_l_A_01 [{shapeType=Capsule, length=12, radius=6, shapeName=dyn_calf_pants_l_A_01_capsule, center=<12 0 2>, rotation=<rot 90 0 0>}]
 dyn_calf_pants_l_A_02 [{shapeType=Capsule, length=12, radius=6, shapeName=dyn_calf_pants_l_A_02_capsule, center=<12 0 2>, rotation=<rot 90 0 0>}]
 dyn_calf_pants_l_B_01 [{shapeType=Capsule, length=12, radius=6, shapeName=dyn_calf_pants_l_B_01_capsule, center=<12 0 2>, rotation=<rot 90 0 0>}]
 dyn_calf_pants_l_B_02 [{shapeType=Capsule, length=12, radius=6, shapeName=dyn_calf_pants_l_B_02_capsule, center=<12 0 2>, rotation=<rot 90 0 0>}]
 dyn_calf_pants_l_C_01 [{shapeType=Capsule, length=12, radius=6, shapeName=dyn_calf_pants_l_C_01_capsule, center=<12 0 2>, rotation=<rot 90 0 0>}]
 dyn_calf_pants_l_C_02 [{shapeType=Capsule, length=12, radius=6, shapeName=dyn_calf_pants_l_C_02_capsule, center=<12 0 2>, rotation=<rot 90 0 0>}]
 dyn_calf_pants_l_D_01 [{shapeType=Capsule, length=12, radius=6, shapeName=dyn_calf_pants_l_D_01_capsule, center=<12 0 2>, rotation=<rot 90 0 0>}]
 dyn_calf_pants_l_D_02 [{shapeType=Capsule, length=12, radius=6, shapeName=dyn_calf_pants_l_D_02_capsule, center=<12 0 2>, rotation=<rot 90 0 0>}]
 foot_l [{
   shapeType=Capsule
   length=10.7521
   radius=6.7308
   shapeName=foot_l_capsule
   center=<-5.5865 4.2632 -0.198>
   rotation=<rot -0.0198 8.8459 89.732>
  }]
 foot_r [{
   shapeType=Capsule
   length=10.7521
   radius=6.7308
   shapeName=foot_l_capsule
   center=<5.5865 -4.2632 0.198>
   rotation=<rot 0.0198 8.8459 -89.732>
  }]
 pelvis [{shapeType=Capsule, length=5.3176, radius=14.0685, shapeName=pelvis_capsule, rotation=<rot 0 90 0>}]
 thigh_l [{shapeType=Capsule, length=32.4284, radius=7.7581, shapeName=thigh_l_capsule, center=<-14.0962 0 0>, rotation=<rot 90 0 0>}]
 thigh_r [{shapeType=Capsule, length=32.4284, radius=7.7581, shapeName=thigh_l_capsule, center=<14.0962 0 0>, rotation=<rot 90 0 0>}]
}
constraints [
 {
  bone1Name=calf_l
  bone2Name=thigh_l
  swing1LimitDegrees=45
  swing1Motion=Locked
  swing2LimitDegrees=45
  swing2Motion=Locked
  twistLimitDegrees=45
  twistMotion=Locked
 }
 {
  bone1Name=thigh_l
  bone2Name=pelvis
  swing1LimitDegrees=45
  swing1Motion=Locked
  swing2LimitDegrees=45
  swing2Motion=Locked
  twistLimitDegrees=45
  twistMotion=Locked
 }
 {
  bone1Name=foot_l
  bone2Name=calf_l
  swing1LimitDegrees=45
  swing1Motion=Locked
  swing2LimitDegrees=45
  swing2Motion=Locked
  twistLimitDegrees=45
  twistMotion=Locked
 }
 {
  bone1Name=dyn_calf_pants_l_A_01
  bone2Name=base_calf_pants_l_A
  swing1LimitDegrees=5
  swing2LimitDegrees=8
  twistLimitDegrees=45
  twistMotion=Locked
 }
 {
  bone1Name=dyn_calf_pants_l_A_02
  bone2Name=dyn_calf_pants_l_A_01
  swing1LimitDegrees=5
  swing2LimitDegrees=8
  twistLimitDegrees=45
  twistMotion=Locked
 }
 {
  bone1Name=dyn_calf_pants_l_B_01
  bone2Name=base_calf_pants_l_B
  swing1LimitDegrees=3
  swing2LimitDegrees=5
  twistLimitDegrees=45
  twistMotion=Locked
 }
 {
  bone1Name=dyn_calf_pants_l_B_02
  bone2Name=dyn_calf_pants_l_B_01
  swing1LimitDegrees=3
  swing2LimitDegrees=5
  twistLimitDegrees=45
  twistMotion=Locked
 }
 {
  bone1Name=dyn_calf_pants_l_C_01
  bone2Name=base_calf_pants_l_C
  swing1LimitDegrees=4
  swing2LimitDegrees=6
  twistLimitDegrees=45
  twistMotion=Locked
 }
 {
  bone1Name=dyn_calf_pants_l_C_02
  bone2Name=dyn_calf_pants_l_C_01
  swing1LimitDegrees=4
  swing2LimitDegrees=6
  twistLimitDegrees=45
  twistMotion=Locked
 }
 {
  bone1Name=dyn_calf_pants_l_D_01
  bone2Name=base_calf_pants_l_D
  swing1LimitDegrees=2
  swing2LimitDegrees=3
  twistLimitDegrees=45
  twistMotion=Locked
 }
 {
  bone1Name=dyn_calf_pants_l_D_02
  bone2Name=dyn_calf_pants_l_D_01
  swing1LimitDegrees=2
  swing2LimitDegrees=3
  twistLimitDegrees=45
  twistMotion=Locked
 }
 {
  bone1Name=base_calf_pants_l_A
  bone2Name=calf_l
  swing1LimitDegrees=45
  swing1Motion=Locked
  swing2LimitDegrees=45
  swing2Motion=Locked
  twistLimitDegrees=45
  twistMotion=Locked
 }
 {
  bone1Name=base_calf_pants_l_B
  bone2Name=calf_l
  swing1LimitDegrees=45
  swing1Motion=Locked
  swing2LimitDegrees=45
  swing2Motion=Locked
  twistLimitDegrees=45
  twistMotion=Locked
 }
 {
  bone1Name=base_calf_pants_l_C
  bone2Name=calf_l
  swing1LimitDegrees=45
  swing1Motion=Locked
  swing2LimitDegrees=45
  swing2Motion=Locked
  twistLimitDegrees=45
  twistMotion=Locked
 }
 {
  bone1Name=base_calf_pants_l_D
  bone2Name=calf_l
  swing1LimitDegrees=45
  swing1Motion=Locked
  swing2LimitDegrees=45
  swing2Motion=Locked
  twistLimitDegrees=45
  twistMotion=Locked
 }
 {
  bone1Name=thigh_r
  bone2Name=pelvis
  swing1LimitDegrees=45
  swing1Motion=Locked
  swing2LimitDegrees=45
  swing2Motion=Locked
  twistLimitDegrees=45
  twistMotion=Locked
 }
 {
  bone1Name=calf_r
  bone2Name=thigh_r
  swing1LimitDegrees=45
  swing1Motion=Locked
  swing2LimitDegrees=45
  swing2Motion=Locked
  twistLimitDegrees=45
  twistMotion=Locked
 }
 {
  bone1Name=foot_r
  bone2Name=calf_r
  swing1LimitDegrees=45
  swing1Motion=Locked
  swing2LimitDegrees=45
  swing2Motion=Locked
  twistLimitDegrees=45
  twistMotion=Locked
 }
]
