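# Post-process-style AnimBP for the LatteStir male cosmetic body, adding secondary motion to a pose copied from the parent mesh.
# A Control Rig feeds two RigidBody chains (calves, spine), with front-end state choosing the spine physics asset.
# AnimDynamics then drives belt bones, and additive bone offsets adjust bandolier and sleeves.
# EventGraph tracks pickaxe swings and computes gravity/RBAN settings, elbow alpha and LOD-gated switches, most not visibly used in AnimGraph.
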AnimBlueprint /BRCosmetics/Characters/Player/Male/Medium/Bodies/M_MED_LatteStir/Meshes/M_MED_LatteStir_AnimBP
target_skeleton=&/M_MED_LatteStir_Skeleton
graphs [
 AnimGraph {nodes [
   Local To Component()
   n1 = RigidBody!(
    ComponentPose=Local To Component(LocalPose=Use cached pose 'CR_Start'())
    Alpha
    OverrideWorldGravity:Vector="0,0,-980"
    ExternalForce
    SimSpaceSettings:Sim Space Settings Structure=(WorldAlpha=1,VelocityScaleZ=0,DampingAlpha=1,MaxLinearVelocity=500,MaxAngularVelocity=3,MaxLinearAcceleration=700,MaxAngularAcceleration=1000,ExternalLinearDragV=(X=0.1,Y=0.1,Z=0),ExternalLinearVelocity=(X=0,Y=0,Z=0),ExternalAngularVelocity=(X=0,Y=0,Z=0))
   ) -> (Pose) {
    bClampLinearTranslationLimitToRefPose=true
    bOverrideWorldGravity=true
    overridePhysicsAsset=&/M_MED_LatteStir_Left_Physics
    simulationSpace=BaseBoneSpace
    alphaBoolBlend {blendOption=Linear}
    baseBoneRef {boneName=calf_l}
    componentAppliedLinearAccClamp=<0 0 0>
    overrideWorldGravity {z=-980}
    simSpaceSettings {
     maxAngularAcceleration=1000
     maxAngularVelocity=3
     maxLinearAcceleration=700
     maxLinearVelocity=500
     velocityScaleZ=0
     worldAlpha=1
     externalLinearDragV=<0.1 0.1>
    }
   }
   n2 = RigidBody(
    ComponentPose=@n1
    OverrideWorldGravity="0,0,-980"
    SimSpaceSettings=(WorldAlpha=1,VelocityScaleZ=0,DampingAlpha=1,MaxLinearVelocity=500,MaxAngularVelocity=3,MaxLinearAcceleration=700,MaxAngularAcceleration=1000,ExternalLinearDragV=(X=0.1,Y=0.1,Z=0),ExternalLinearVelocity=(X=0,Y=0,Z=0),ExternalAngularVelocity=(X=0,Y=0,Z=0))
   ) {
    bClampLinearTranslationLimitToRefPose=true
    bOverrideWorldGravity=true
    overridePhysicsAsset=&/M_MED_LatteStir_Right_Physics
    simulationSpace=BaseBoneSpace
    alphaBoolBlend {blendOption=Linear}
    baseBoneRef {boneName=calf_r}
    componentAppliedLinearAccClamp=<0 0 0>
    overrideWorldGravity {z=-980}
    simSpaceSettings {
     maxAngularAcceleration=1000
     maxAngularVelocity=3
     maxLinearAcceleration=700
     maxLinearVelocity=500
     velocityScaleZ=0
     worldAlpha=1
     externalLinearDragV=<0.1 0.1>
    }
   }
   n3 = RigidBody(
    ComponentPose=@n2
    OverrideWorldGravity="0,0,-980"
    SimSpaceSettings=(WorldAlpha=1,VelocityScaleZ=0,DampingAlpha=1,MaxLinearVelocity=500,MaxAngularVelocity=3,MaxLinearAcceleration=700,MaxAngularAcceleration=1000,ExternalLinearDragV=(X=0.1,Y=0.1,Z=0),ExternalLinearVelocity=(X=0,Y=0,Z=0),ExternalAngularVelocity=(X=0,Y=0,Z=0))
   ) {
    bClampLinearTranslationLimitToRefPose=true
    bOverrideWorldGravity=true
    overridePhysicsAsset=&/M_MED_LatteStir_FE_Physics
    simulationSpace=BaseBoneSpace
    alphaBoolBlend {blendOption=Linear}
    baseBoneRef {boneName=spine_05}
    componentAppliedLinearAccClamp=<0 0 0>
    overrideWorldGravity {z=-980}
    simSpaceSettings {
     maxAngularAcceleration=1000
     maxAngularVelocity=3
     maxLinearAcceleration=700
     maxLinearVelocity=500
     velocityScaleZ=0
     worldAlpha=1
     externalLinearDragV=<0.1 0.1>
    }
   }
   n4 = Component To Local(ComponentPose=@n3)
   n5 = RigidBody(
    ComponentPose=Local To Component(LocalPose=Use cached pose 'CR_Start'())
    OverrideWorldGravity="0,0,-980"
    SimSpaceSettings=(WorldAlpha=1,VelocityScaleZ=0,DampingAlpha=1,MaxLinearVelocity=500,MaxAngularVelocity=3,MaxLinearAcceleration=700,MaxAngularAcceleration=1000,ExternalLinearDragV=(X=0.1,Y=0.1,Z=0),ExternalLinearVelocity=(X=0,Y=0,Z=0),ExternalAngularVelocity=(X=0,Y=0,Z=0))
   ) {
    bClampLinearTranslationLimitToRefPose=true
    bOverrideWorldGravity=true
    overridePhysicsAsset=&M_MED_LatteStir_Left_Physics
    simulationSpace=BaseBoneSpace
    alphaBoolBlend {blendOption=Linear}
    baseBoneRef {boneName=calf_l}
    componentAppliedLinearAccClamp=<0 0 0>
    overrideWorldGravity {z=-980}
    simSpaceSettings {
     maxAngularAcceleration=1000
     maxAngularVelocity=3
     maxLinearAcceleration=700
     maxLinearVelocity=500
     velocityScaleZ=0
     worldAlpha=1
     externalLinearDragV=<0.1 0.1>
    }
   }
   n6 = RigidBody(
    ComponentPose=@n5
    OverrideWorldGravity="0,0,-980"
    SimSpaceSettings=(WorldAlpha=1,VelocityScaleZ=0,DampingAlpha=1,MaxLinearVelocity=500,MaxAngularVelocity=3,MaxLinearAcceleration=700,MaxAngularAcceleration=1000,ExternalLinearDragV=(X=0.1,Y=0.1,Z=0),ExternalLinearVelocity=(X=0,Y=0,Z=0),ExternalAngularVelocity=(X=0,Y=0,Z=0))
   ) {
    bClampLinearTranslationLimitToRefPose=true
    bOverrideWorldGravity=true
    overridePhysicsAsset=&M_MED_LatteStir_Right_Physics
    simulationSpace=BaseBoneSpace
    alphaBoolBlend {blendOption=Linear}
    baseBoneRef {boneName=calf_r}
    componentAppliedLinearAccClamp=<0 0 0>
    overrideWorldGravity {z=-980}
    simSpaceSettings {
     maxAngularAcceleration=1000
     maxAngularVelocity=3
     maxLinearAcceleration=700
     maxLinearVelocity=500
     velocityScaleZ=0
     worldAlpha=1
     externalLinearDragV=<0.1 0.1>
    }
   }
   n7 = RigidBody(
    ComponentPose=@n6
    OverrideWorldGravity="0,0,-980"
    SimSpaceSettings=(WorldAlpha=1,VelocityScaleZ=0,DampingAlpha=1,MaxLinearVelocity=500,MaxAngularVelocity=3,MaxLinearAcceleration=700,MaxAngularAcceleration=1000,ExternalLinearDragV=(X=0.1,Y=0.1,Z=0),ExternalLinearVelocity=(X=0,Y=0,Z=0),ExternalAngularVelocity=(X=0,Y=0,Z=0))
   ) {
    bClampLinearTranslationLimitToRefPose=true
    bOverrideWorldGravity=true
    overridePhysicsAsset=&/M_MED_LatteStir_Physics
    simulationSpace=BaseBoneSpace
    alphaBoolBlend {blendOption=Linear}
    baseBoneRef {boneName=spine_05}
    componentAppliedLinearAccClamp=<0 0 0>
    overrideWorldGravity {z=-980}
    simSpaceSettings {
     maxAngularAcceleration=1000
     maxAngularVelocity=3
     maxLinearAcceleration=700
     maxLinearVelocity=500
     velocityScaleZ=0
     worldAlpha=1
     externalLinearDragV=<0.1 0.1>
    }
   }
   n8 = Component To Local(ComponentPose=@n7)
   n9 = Get bIsInFrontEnd()
   n10 = Anim Dynamics - Bone: dyn_belt_l_01!(
    ComponentPose=Local To Component(LocalPose=Blend Poses by bool(BlendPose_0=@n4, BlendPose_1=@n8, bActiveValue=@n9))
    Alpha=1
    GravityOverride:Vector="0,0,-980"
    ExternalForce="0,0,0"
   ) -> (Pose) {
    angularDampingOverride=1
    angularSpringConstant=12000
    bAngularSpring=true
    bOverrideAngularDamping=true
    bUseGravityOverride=true
    gravityScale=0.35
    simulationSpace=BoneRelative
    boundBone {boneName=dyn_belt_l_01}
    componentAppliedLinearAccClamp=<10000 10000 10000>
    componentLinearAccScale=<1 1 1>
    componentLinearVelScale=<1 1 1>
    gravityOverride {z=-980}
    physicsBodyDefinitions [{
      boundBone {boneName=dyn_belt_l_01}
      boxExtents=<20 30 30>
      constraintSetup {angularLimitsMax=<1 2 10>, angularLimitsMin=<-1 -20 -10>, angularTarget=<1 0 0>}
      localJointOffset=<4 0 0>
     }]
    relativeSpaceBone {boneName=pelvis}
    simSpaceSettings {maxAngularAcceleration=75, maxAngularVelocity=15, simSpaceAngularAlpha=1}
   }
   n11 = Anim Dynamics - Bone: dyn_belt_l_02!(ComponentPose=@n10, Alpha=1, GravityOverride:Vector="0,0,-980", ExternalForce="0,0,0") -> (Pose) {
    angularDampingOverride=1
    angularSpringConstant=12000
    bAngularSpring=true
    bOverrideAngularDamping=true
    bUseGravityOverride=true
    gravityScale=0.35
    simulationSpace=BoneRelative
    boundBone {boneName=dyn_belt_l_02}
    componentAppliedLinearAccClamp=<10000 10000 10000>
    componentLinearAccScale=<1 1 1>
    componentLinearVelScale=<1 1 1>
    gravityOverride {z=-980}
    physicsBodyDefinitions [{
      boundBone {boneName=dyn_belt_l_02}
      boxExtents=<20 30 30>
      constraintSetup {angularLimitsMax=<1 2 10>, angularLimitsMin=<-1 -20 -10>, angularTarget=<1 0 0>}
      localJointOffset=<4 0 0>
     }]
    relativeSpaceBone {boneName=pelvis}
    simSpaceSettings {maxAngularAcceleration=75, maxAngularVelocity=15, simSpaceAngularAlpha=1}
   }
   n12 = Anim Dynamics - Bone: dyn_belt_l_03!(ComponentPose=@n11, Alpha=1, GravityOverride:Vector="0,0,-980", ExternalForce="0,0,0") -> (Pose) {
    angularDampingOverride=1
    angularSpringConstant=12000
    bAngularSpring=true
    bOverrideAngularDamping=true
    bUseGravityOverride=true
    gravityScale=0.35
    simulationSpace=BoneRelative
    boundBone {boneName=dyn_belt_l_03}
    componentAppliedLinearAccClamp=<10000 10000 10000>
    componentLinearAccScale=<1 1 1>
    componentLinearVelScale=<1 1 1>
    gravityOverride {z=-980}
    physicsBodyDefinitions [{
      boundBone {boneName=dyn_belt_l_03}
      boxExtents=<20 30 30>
      constraintSetup {angularLimitsMax=<1 2 10>, angularLimitsMin=<-1 -20 -10>, angularTarget=<1 0 0>}
      localJointOffset=<4 0 0>
     }]
    relativeSpaceBone {boneName=pelvis}
    simSpaceSettings {maxAngularAcceleration=75, maxAngularVelocity=15, simSpaceAngularAlpha=1}
   }
   n13 = Anim Dynamics - Bone: dyn_belt_r_01!(ComponentPose=@n12, Alpha=1, GravityOverride:Vector="0,0,-980", ExternalForce="0,0,0") -> (Pose) {
    angularDampingOverride=1
    angularSpringConstant=12000
    bAngularSpring=true
    bOverrideAngularDamping=true
    bUseGravityOverride=true
    gravityScale=0.35
    simulationSpace=BoneRelative
    boundBone {boneName=dyn_belt_r_01}
    componentAppliedLinearAccClamp=<10000 10000 10000>
    componentLinearAccScale=<1 1 1>
    componentLinearVelScale=<1 1 1>
    gravityOverride {z=-980}
    physicsBodyDefinitions [{
      boundBone {boneName=dyn_belt_r_01}
      boxExtents=<20 30 30>
      constraintSetup {angularLimitsMax=<1 2 10>, angularLimitsMin=<-1 -20 -10>, angularTarget=<1 0 0>}
      localJointOffset=<4 0 0>
     }]
    relativeSpaceBone {boneName=pelvis}
    simSpaceSettings {maxAngularAcceleration=75, maxAngularVelocity=15, simSpaceAngularAlpha=1}
   }
   n14 = Anim Dynamics - Bone: dyn_belt_r_02!(ComponentPose=@n13, Alpha=1, GravityOverride:Vector="0,0,-980", ExternalForce="0,0,0") -> (Pose) {
    angularDampingOverride=1
    angularSpringConstant=12000
    bAngularSpring=true
    bOverrideAngularDamping=true
    bUseGravityOverride=true
    gravityScale=0.35
    simulationSpace=BoneRelative
    boundBone {boneName=dyn_belt_r_02}
    componentAppliedLinearAccClamp=<10000 10000 10000>
    componentLinearAccScale=<1 1 1>
    componentLinearVelScale=<1 1 1>
    gravityOverride {z=-980}
    physicsBodyDefinitions [{
      boundBone {boneName=dyn_belt_r_02}
      boxExtents=<20 30 30>
      constraintSetup {angularLimitsMax=<1 2 10>, angularLimitsMin=<-1 -20 -10>, angularTarget=<1 0 0>}
      localJointOffset=<4 0 0>
     }]
    relativeSpaceBone {boneName=pelvis}
    simSpaceSettings {maxAngularAcceleration=75, maxAngularVelocity=15, simSpaceAngularAlpha=1}
   }
   n15 = Anim Dynamics - Bone: dyn_belt_r_03!(ComponentPose=@n14, Alpha=1, GravityOverride:Vector="0,0,-980", ExternalForce="0,0,0") -> (Pose) {
    angularDampingOverride=1
    angularSpringConstant=12000
    bAngularSpring=true
    bOverrideAngularDamping=true
    bUseGravityOverride=true
    gravityScale=0.35
    simulationSpace=BoneRelative
    boundBone {boneName=dyn_belt_r_03}
    componentAppliedLinearAccClamp=<10000 10000 10000>
    componentLinearAccScale=<1 1 1>
    componentLinearVelScale=<1 1 1>
    gravityOverride {z=-980}
    physicsBodyDefinitions [{
      boundBone {boneName=dyn_belt_r_03}
      boxExtents=<20 30 30>
      constraintSetup {angularLimitsMax=<1 2 10>, angularLimitsMin=<-1 -20 -10>, angularTarget=<1 0 0>}
      localJointOffset=<4 0 0>
     }]
    relativeSpaceBone {boneName=pelvis}
    simSpaceSettings {maxAngularAcceleration=75, maxAngularVelocity=15, simSpaceAngularAlpha=1}
   }
   n16 = "Transform (Modify) Bone - Bone: dyn_bandolier_fr_04_l"(ComponentPose=@n15, Alpha=1, Translation="0,0.5,0", Rotation="0,0,0", Scale="1,1,1") {
    rotationMode=BMM_Additive
    rotationSpace=BCS_BoneSpace
    translationMode=BMM_Additive
    translationSpace=BCS_BoneSpace
    boneToModify {boneName=dyn_bandolier_fr_04_l}
    translation {y=0.5}
   }
   n17 = "Transform (Modify) Bone - Bone: dyn_bandolier_fr_03_l"(ComponentPose=@n16, Alpha=1, Translation="0,1,0", Rotation="0,0,0", Scale="1,1,1") {
    rotationMode=BMM_Additive
    rotationSpace=BCS_BoneSpace
    translationMode=BMM_Additive
    translationSpace=BCS_BoneSpace
    alphaScaleBiasClamp {bInterpResult=true, interpSpeedDecreasing=15, interpSpeedIncreasing=5}
    boneToModify {boneName=dyn_bandolier_fr_03_l}
    translation {y=1}
   }
   n18 = "Transform (Modify) Bone - Bone: dyn_bandolier_bk_05_l"(ComponentPose=@n17, Alpha=1, Translation="1,0,0", Rotation="0,0,0", Scale="1,1,1") {
    rotationMode=BMM_Additive
    rotationSpace=BCS_BoneSpace
    translationMode=BMM_Additive
    translationSpace=BCS_BoneSpace
    alphaScaleBiasClamp {bClampResult=true, bInterpResult=true}
    boneToModify {boneName=dyn_bandolier_bk_05_l}
    translation {x=1}
   }
   n19 = "Transform (Modify) Bone - Bone: dyn_bandolier_fr_03_l"(ComponentPose=@n18, Alpha=1, Translation="0,1,0", Rotation="0,0,0", Scale="1,1,1") {
    rotationMode=BMM_Additive
    rotationSpace=BCS_BoneSpace
    translationMode=BMM_Additive
    translationSpace=BCS_BoneSpace
    alphaScaleBiasClamp {bInterpResult=true, interpSpeedDecreasing=15, interpSpeedIncreasing=5}
    boneToModify {boneName=dyn_bandolier_fr_03_l}
    translation {y=1}
   }
   n20 = "Transform (Modify) Bone - Bone: dyn_bandolier_fr_04_l"(ComponentPose=@n19, Alpha=1, Translation="0,0,0", Rotation="-5,4,0", Scale="1,1,1") {
    rotationMode=BMM_Additive
    rotationSpace=BCS_BoneSpace
    translationMode=BMM_Additive
    translationSpace=BCS_BoneSpace
    boneToModify {boneName=dyn_bandolier_fr_04_l}
    rotation {pitch=-5, yaw=4}
   }
   n21 = "Transform (Modify) Bone - Bone: dyn_bandolier_bk_04_l"(ComponentPose=@n20, Alpha=1, Translation="0,0,0", Rotation="0,-5,0", Scale="1,1,1") {
    rotationMode=BMM_Additive
    rotationSpace=BCS_BoneSpace
    translationMode=BMM_Additive
    translationSpace=BCS_BoneSpace
    alphaScaleBiasClamp {bInterpResult=true, interpSpeedDecreasing=15, interpSpeedIncreasing=5}
    boneToModify {boneName=dyn_bandolier_bk_04_l}
    rotation {yaw=-5}
   }
   n22 = "Transform (Modify) Bone - Bone: dyn_bandolier_fr_03_l"(ComponentPose=@n21, Alpha=1, Translation="0,0.5,0", Rotation="0,0,0", Scale="1,1,1") {
    rotationMode=BMM_Additive
    rotationSpace=BCS_BoneSpace
    translationMode=BMM_Additive
    translationSpace=BCS_BoneSpace
    boneToModify {boneName=dyn_bandolier_fr_03_l}
    translation {y=0.5}
   }
   n23 = "Transform (Modify) Bone - Bone: dyn_bandolier_fr_05_l"(ComponentPose=@n22, Alpha=1, Translation="0,0,0", Rotation="20,0,0", Scale="1,1,1") {
    rotationMode=BMM_Additive
    rotationSpace=BCS_BoneSpace
    translationMode=BMM_Additive
    translationSpace=BCS_BoneSpace
    boneToModify {boneName=dyn_bandolier_fr_05_l}
    rotation {pitch=20}
   }
   n24 = "Transform (Modify) Bone - Bone: dyn_bandolier_fr_03_l"(ComponentPose=@n23, Alpha=1, Translation="0,0,0", Rotation="-4,4,0", Scale="1,1,1") {
    rotationMode=BMM_Additive
    rotationSpace=BCS_BoneSpace
    translationMode=BMM_Additive
    translationSpace=BCS_BoneSpace
    boneToModify {boneName=dyn_bandolier_fr_03_l}
    rotation {pitch=-4, yaw=4}
   }
   n25 = "Transform (Modify) Bone - Bone: dyn_bandolier_bk_04_l"(ComponentPose=@n24, Alpha=1, Translation="0,0,-0.5", Rotation="-6,0,0", Scale="1,1,1") {
    rotationMode=BMM_Additive
    rotationSpace=BCS_BoneSpace
    translationMode=BMM_Additive
    translationSpace=BCS_BoneSpace
    alphaScaleBiasClamp {bInterpResult=true, interpSpeedDecreasing=15, interpSpeedIncreasing=5}
    boneToModify {boneName=dyn_bandolier_bk_04_l}
    rotation {pitch=-6}
    translation {z=-0.5}
   }
   n26 = "Transform (Modify) Bone - Bone: dyn_bandolier_fr_05_l"(ComponentPose=@n25, Alpha=1, Translation="0,0,0", Rotation="15,0,0", Scale="1,1,1") {
    rotationMode=BMM_Additive
    rotationSpace=BCS_BoneSpace
    translationMode=BMM_Additive
    translationSpace=BCS_BoneSpace
    boneToModify {boneName=dyn_bandolier_fr_05_l}
    rotation {pitch=15}
   }
   n27 = "Transform (Modify) Bone - Bone: dyn_bandolier_bk_05_l"(ComponentPose=@n26, Alpha=1, Translation="0,0,-1", Rotation="15,0,0", Scale="1,1,1") {
    rotationMode=BMM_Additive
    rotationSpace=BCS_BoneSpace
    translationMode=BMM_Additive
    translationSpace=BCS_BoneSpace
    alphaScaleBiasClamp {bInterpResult=true, interpSpeedDecreasing=15, interpSpeedIncreasing=5}
    boneToModify {boneName=dyn_bandolier_bk_05_l}
    rotation {pitch=15}
    translation {z=-1}
   }
   n28 = "Transform (Modify) Bone - Bone: dyn_sleeve_r_D"(ComponentPose=@n27, Alpha=1, Translation="4,0,-7", Rotation="20,0,0", Scale="1,1,1") {
    rotationMode=BMM_Additive
    rotationSpace=BCS_BoneSpace
    translationMode=BMM_Additive
    translationSpace=BCS_BoneSpace
    boneToModify {boneName=dyn_sleeve_r_D}
    rotation {pitch=20}
    translation {x=4, z=-7}
   }
   n29 = "Transform (Modify) Bone - Bone: base_sleeve_l_B"(ComponentPose=@n28, Alpha=1, Translation="-2,0,2", Rotation="0,0,0", Scale="1,1,1") {
    rotationMode=BMM_Additive
    rotationSpace=BCS_BoneSpace
    translationMode=BMM_Additive
    translationSpace=BCS_BoneSpace
    boneToModify {boneName=base_sleeve_l_B}
    translation {x=-2, z=2}
   }
   n30 = "Transform (Modify) Bone - Bone: base_sleeve_l_A"(ComponentPose=@n29, Alpha=1, Translation="-2,-2,3", Rotation="0,0,0", Scale="1,1,1") {
    rotationMode=BMM_Additive
    rotationSpace=BCS_BoneSpace
    translationMode=BMM_Additive
    translationSpace=BCS_BoneSpace
    boneToModify {boneName=base_sleeve_l_A}
    translation=<-2 -2 3>
   }
   n31 = "Transform (Modify) Bone - Bone: base_sleeve_l_C"(ComponentPose=@n30, Alpha=1, Translation="-2,2,1", Rotation="0,0,0", Scale="1,1,1") {
    rotationMode=BMM_Additive
    rotationSpace=BCS_BoneSpace
    translationMode=BMM_Additive
    translationSpace=BCS_BoneSpace
    boneToModify {boneName=base_sleeve_l_C}
    translation=<-2 2 1>
   }
   n32 = "Transform (Modify) Bone - Bone: base_sleeve_r_B"(ComponentPose=@n31, Alpha=1, Translation="-3,0,2", Rotation="0,0,0", Scale="1,1,1") {
    rotationMode=BMM_Additive
    rotationSpace=BCS_BoneSpace
    translationMode=BMM_Additive
    translationSpace=BCS_BoneSpace
    boneToModify {boneName=base_sleeve_r_B}
    translation {x=-3, z=2}
   }
   n33 = "Transform (Modify) Bone - Bone: base_sleeve_l_B"(ComponentPose=@n32, Alpha=1, Translation="-4,0,4", Rotation="0,0,0", Scale="1,1,1") {
    rotationMode=BMM_Additive
    rotationSpace=BCS_BoneSpace
    translationMode=BMM_Additive
    translationSpace=BCS_BoneSpace
    boneToModify {boneName=base_sleeve_l_B}
    translation {x=-4, z=4}
   }
   n34 = "Transform (Modify) Bone - Bone: base_sleeve_l_A"(ComponentPose=@n33, Alpha=1, Translation="-2,0,2", Rotation="0,0,0", Scale="1,1,1") {
    rotationMode=BMM_Additive
    rotationSpace=BCS_BoneSpace
    translationMode=BMM_Additive
    translationSpace=BCS_BoneSpace
    boneToModify {boneName=base_sleeve_l_A}
    translation {x=-2, z=2}
   }
   n35 = "Transform (Modify) Bone - Bone: base_sleeve_l_C"(ComponentPose=@n34, Alpha=1, Translation="-2,0,2", Rotation="0,0,0", Scale="1,1,1") {
    rotationMode=BMM_Additive
    rotationSpace=BCS_BoneSpace
    translationMode=BMM_Additive
    translationSpace=BCS_BoneSpace
    boneToModify {boneName=base_sleeve_l_C}
    translation {x=-2, z=2}
   }
   n36 = "Transform (Modify) Bone - Bone: base_sleeve_l_B"(ComponentPose=@n35, Alpha=1, Translation="-1,0,1", Rotation="0,0,0", Scale="1,1,1") {
    rotationMode=BMM_Additive
    rotationSpace=BCS_BoneSpace
    translationMode=BMM_Additive
    translationSpace=BCS_BoneSpace
    boneToModify {boneName=base_sleeve_l_B}
    translation {x=-1, z=1}
   }
   n37 = "Transform (Modify) Bone - Bone: base_sleeve_r_B"(ComponentPose=@n36, Alpha=1, Translation="-2,0,2", Rotation="0,0,0", Scale="1,1,1") {
    rotationMode=BMM_Additive
    rotationSpace=BCS_BoneSpace
    translationMode=BMM_Additive
    translationSpace=BCS_BoneSpace
    boneToModify {boneName=base_sleeve_r_B}
    translation {x=-2, z=2}
   }
   Output Pose(Result=Component To Local(ComponentPose=@n37)) {layerGroup=None}
   n38 = Control Rig!(
    Source=Copy Pose From Mesh() {bUseAttachedParent=true}
    Alpha
    isPlayingEmote:bool=Get bIsPlayingEmote()
    isPickaxeSwing:bool=Get isPickaxeSwing() -> (isPickaxeSwing:bool)
    isFrontEnd:bool=Get bIsInFrontEnd()
   ) -> (Pose) {controlRigAssetReference {blueprintRigClass=&/M_MED_LatteStir_CtrlRig.M_MED_LatteStir_CtrlRig_C}}
   Save cached pose 'CR_Start'(Pose=@n38) {cacheName=CR_Start}
  ]}
 EventGraph {nodes [
   Event BlueprintUpdateAnimation()
   > n1 = Sequence!(execute) -> (then_0, then_1, then_2:Exec)
   > n2 = "Gravity Override State Function (Parameter Asset) (Nativized)"(
    CharacterPartAnimInstance=Self-Reference()
    GameTimeInSeconds=GetGameTimeInSeconds()
    DebugJointName=spine_05
    Parameters=&/M_MED_LatteStir_Parameters
   )
   > Set Gravity Override(Gravity Override=@n2)
   > Set RBAN Settings Out(RBAN Settings Out=@n2.OutRigidBodyAnimNodeParameters)
   n3 = GetDeltaTransformFromRefPose(self=GetOwnerSkeletalMeshComponent(), BoneName=lowerarm_r, BaseName=upperarm_r)
   n4 = "Absolute (Float)"(A=BreakRotator(InRot=BreakTransform(InTransform=@n3).Rotation).Yaw)
   Set R_Elbow_Up_Alpha(
    execute=@n1.then_1
    R_Elbow_Up_Alpha:double=float * float(A=MapRangeClamped(Value=@n4, InRangeA=0, InRangeB=80), B:bool=Get bIsInFrontEnd()) {is=title K2Node_PromotableOperator}
   ) -> (Output_Get:double)
   n5 = Get CurrentLOD()
   n6 = Set LOD1_Active(execute=@n1.then_2, LOD1_Active:bool="integer >= integer"(A=@n5, B=1)) -> (Output_Get:bool)
   > n7 = Set SwitchPlatform(SwitchPlatform:bool=@n6) -> (Output_Get:bool)
   > Set SwitchSkydive(SwitchSkydive:bool=AND Boolean(A=@n7, B=Get bIsSkyDiving())) -> (Output_Get:bool)
   > Set SwitchTacticalSprint(SwitchTacticalSprint:bool=AND Boolean(A=@n7, B=Get bIsTacticalSprinting())) -> (Output_Get:bool)
   Event OnBeginPickaxeSwing() {is=title K2Node_Event}
   > Set isPickaxeSwing(isPickaxeSwing:bool=true) -> (Output_Get:bool)
   Event OnEndPickaxeSwing() {is=title K2Node_Event}
   > Set isPickaxeSwing(isPickaxeSwing=false)
   "Not Equal (Integer)"(A=@n5)
  ]}
]
variables [
 Gravity Override {type=Vector}
 isPickaxeSwing {type=boolean}
 LOD1_Active {type=boolean}
 R_Elbow_Up_Alpha {type=number}
 RBAN Settings Out {type=FortRigidBodyAnimNodeParameters}
 SwitchPlatform {type=boolean}
 SwitchSkydive {type=boolean}
 SwitchTacticalSprint {type=boolean}
]
